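; Marker actor for a mythic spawner location in the Rift Gauntlets gameplay, with a hotfixable enable switch.
; IsLocationEnabled reads the SpawnLocationEnable scalable float as a bool.
; On BeginPlay, a dedicated server waits for playlist data to be ready, then logs whether the location is enabled; any other instance destroys the actor.
; The spawning itself is not implemented in this graph.
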(event EventBeginPlay
  ;[0.02443 0.15 0.04497 0.5] BeginPlay: Log if location is enabled.
  (|CallParentFunction)
  (if (Networking|IsDedicatedServer)
    (Class|FortAsyncActionCurrentPlaylistReady|PlaylistDataReadyAsync
      (:OnReady
        (Fortnite|FortShippingLog self (Utilities|String|Append "BeginPlay: Location enabled? " (Utilities|String|ToString(Boolean) (CallFunction|IsLocationEnabled))))))
    (else
      (Actor|DestroyActor))))

(fn IsLocationEnabled
  (return (ScalableFloat|GetValueasBool (Variables|Default|GetHotfix_SpawnLocationEnable))))

(fn ConstructionScript)
